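; Game state component for the Spire Boss mode that hands out spire loot groups in a randomized rotation.
; On BeginPlay it resets an iterator and shuffles the loot group array.
; GetNextSpireLootGroup returns the entry at the iterator, advances it and wraps to zero at the end, so every entry is used before any repeats.
; The array is not reshuffled on wrap, so the same order repeats each cycle.
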(event EventBeginPlay
  (CallFunction|InitComponentVariables))

(fn GetNextSpireLootGroup
  (LocalVariables|SetCurrentSpireLootGroup (Utilities|Array|Get(acopy) (Variables|InternalData|GetSpireLootGroupArray) (Variables|InternalData|GetLootArrayIter)))
  (bind _result (Math|Integer|IncrementInt (Variables|InternalData|GetLootArrayIter))
    (:%20%20
      (if (>= _result (Utilities|Array|Length (Variables|InternalData|GetSpireLootGroupArray)))
        (Variables|InternalData|SetLootArrayIter 0))
      (return (LocalVariables|GetCurrentSpireLootGroup)))))

(fn InitComponentVariables
  (Variables|InternalData|SetLootArrayIter 0)
  (Utilities|Array|Shuffle (Variables|InternalData|GetSpireLootGroupArray)))
